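# Physics asset of capsule colliders for the lower body and torso of the M_MED_Blockstack_Lynx player cosmetic body.
# One capsule covers each of the pelvis, spine_02 through spine_05, thighs, calves and deform_foot bones, leaving arms and head without bodies.
# Constraints chain the spine and both legs to the pelvis, with swing and twist explicitly free at every joint except the knees.
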PhysicsAsset /BRCosmetics/Characters/Player/Male/Medium/Bodies/M_MED_Blockstack_Lynx/Meshes/M_MED_BlockStack_Lynx_Physics_Colliders
bodies {
 calf_l [{shapeType=Capsule, length=32.5604, radius=8.8008, center=<-25.2471 3.1372 0.572>, rotation=<rot 0 79.43 90.008>}]
 calf_r [{shapeType=Capsule, length=32.5604, radius=8.8008, center=<25.2471 -3.1372 -0.572>, rotation=<rot 0 79.43 -89.992>}]
 deform_foot_l [{shapeType=Capsule, length=13.2767, radius=10.1859, shapeName=deform_foot_l_capsule, center=<-2.0269 6.1856 0>, rotation=<rot 90 -90 0>}]
 deform_foot_r [{shapeType=Capsule, length=13.2767, radius=10.1859, shapeName=deform_foot_r_capsule, center=<2.0269 -6.1856 0>, rotation=<rot 90 90 0>}]
 pelvis [{shapeType=Capsule, length=13.5432, radius=16.4568, shapeName=pelvis_capsule}]
 spine_02 [{shapeType=Capsule, length=10.7245, radius=18, shapeName=spine_02_capsule, center=<0 -0.5485 0>}]
 spine_03 [{shapeType=Capsule, length=25, radius=17, shapeName=spine_03_capsule, center=<0 -0.5485 0>}]
 spine_04 [{shapeType=Capsule, length=25, radius=17.5, shapeName=spine_04_capsule, center=<0 -0.5485 0>}]
 spine_05 [{shapeType=Capsule, length=25, radius=17, shapeName=spine_05_capsule, rotation=<rot 0 -90 0>}]
 thigh_l [{shapeType=Capsule, length=33.9361, radius=7.7193, shapeName=thigh_l_capsule, center=<-19.6819 0 0>, rotation=<rot 90 180 0>}]
 thigh_r [{shapeType=Capsule, length=33.9361, radius=7.7193, shapeName=thigh_r_capsule, center=<19.6819 0 0>, rotation=<rot 90 0 0>}]
}
constraints [
 {
  bone1Name=spine_03
  bone2Name=spine_02
  swing1LimitDegrees=45
  swing1Motion=Free
  swing2LimitDegrees=45
  swing2Motion=Free
  twistLimitDegrees=45
  twistMotion=Free
 }
 {
  bone1Name=spine_04
  bone2Name=spine_03
  swing1LimitDegrees=45
  swing1Motion=Free
  swing2LimitDegrees=45
  swing2Motion=Free
  twistLimitDegrees=45
  twistMotion=Free
 }
 {bone1Name=calf_l, bone2Name=thigh_l, swing1LimitDegrees=45, swing2LimitDegrees=45, twistLimitDegrees=45}
 {bone1Name=calf_r, bone2Name=thigh_r, swing1LimitDegrees=45, swing2LimitDegrees=45, twistLimitDegrees=45}
 {
  bone1Name=deform_foot_r
  bone2Name=calf_r
  swing1LimitDegrees=45
  swing1Motion=Free
  swing2LimitDegrees=45
  swing2Motion=Free
  twistLimitDegrees=45
  twistMotion=Free
 }
 {
  bone1Name=deform_foot_l
  bone2Name=calf_l
  swing1LimitDegrees=45
  swing1Motion=Free
  swing2LimitDegrees=45
  swing2Motion=Free
  twistLimitDegrees=45
  twistMotion=Free
 }
 {
  bone1Name=spine_02
  bone2Name=pelvis
  swing1LimitDegrees=45
  swing1Motion=Free
  swing2LimitDegrees=45
  swing2Motion=Free
  twistLimitDegrees=45
  twistMotion=Free
 }
 {
  bone1Name=thigh_l
  bone2Name=pelvis
  swing1LimitDegrees=45
  swing1Motion=Free
  swing2LimitDegrees=45
  swing2Motion=Free
  twistLimitDegrees=45
  twistMotion=Free
 }
 {
  bone1Name=thigh_r
  bone2Name=pelvis
  swing1LimitDegrees=45
  swing1Motion=Free
  swing2LimitDegrees=45
  swing2Motion=Free
  twistLimitDegrees=45
  twistMotion=Free
 }
 {
  bone1Name=spine_05
  bone2Name=spine_04
  swing1LimitDegrees=45
  swing1Motion=Free
  swing2LimitDegrees=45
  swing2Motion=Free
  twistLimitDegrees=45
  twistMotion=Free
 }
]
